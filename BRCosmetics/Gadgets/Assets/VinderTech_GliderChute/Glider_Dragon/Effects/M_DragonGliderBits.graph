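# Unlit translucent particle sprite material, named for the bits effect of the Dragon glider.
# Particle color is normalized, squared five times and rescaled by its original length, then blended toward the raw color by a starburst texture's green channel and passed through MF_FX_EmissiveAdjustments.
# An 8x8 husk-bits flipbook advances with particle relative time, its alpha multiplied by particle alpha.
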Material /BRCosmetics/Gadgets/Assets/VinderTech_GliderChute/Glider_Dragon/Effects/M_DragonGliderBits
graph {
 properties {
  bEnableResponsiveAA=true
  bUsedWithParticleSprites=true
  blendMode=BLEND_Translucent
  pixelDepthOffsetMode=PDOM_Legacy
  shadingModel=MSM_Unlit
  usedShadingModels=Unlit
 }
 nodes [
  Constant() {r=8}
  n1 = ParticleColor()
  n2 = MaterialFunctionCall(Vector 3=@n1.RGB) {materialFunction=/Engine/Functions/Engine_MaterialFunctions02/Utility/VectorLength}
  n3 = Divide(A=@n1.RGB, B=Max(A=@n2.V3 Length) {constB=0.001})
  n4 = Multiply(A=@n3, B=@n3)
  n5 = Multiply(A=@n4, B=@n4)
  n6 = Multiply(A=@n5, B=@n5)
  n7 = Multiply(A=@n6, B=@n6)
  n8 = Multiply(A=@n7, B=@n7)
  MaterialFunctionCall(Emissive=LinearInterpolate(A=Multiply(A=@n8, B=@n2.V3 Length), B=@n1.RGB, Alpha=TextureSample() {texture=/Game/Effects/Fort_Effects/Textures/Flares/T_Starburst}.G)) {materialFunction=&MF_FX_EmissiveAdjustments}
  n9 = ParticleRelativeTime()
  n10 = Constant() {r=64}
  Multiply(
   A=@n1.A
   B=MaterialFunctionCall!(
    Texture=TextureObject() {texture=/Game/Effects/Fort_Effects/Textures/Husk/T_Husk_Bits_SubUV}
    UVs
    SubImages=Constant2Vector() {g=8, r=8}
    Frame=Multiply(A=@n9, B=@n10)
   ) {materialFunction=/Engine/Functions/Engine_MaterialFunctions03/Texturing/SubUV_Function}.Alpha
  )
 ]
}
